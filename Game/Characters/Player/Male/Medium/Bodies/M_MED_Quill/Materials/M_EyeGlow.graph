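# Translucent, two-sided unlit eye glow material for the M_MED_Quill player body.
# Two panning plasma samples blend the glow color, with HDR intensity rising with pixel depth to stay bright at range.
# Opacity combines vertex-red and UV fades with panning noise, plus depth fade off mobile.
# Noise and sines drive vertex-red-masked WPO; the Spline? switch substitutes SplineThicken with jiggle, boosted intensity and constant opacity, leaving spline fade nodes unwired.
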Material /Game/Characters/Player/Male/Medium/Bodies/M_MED_Quill/Materials/M_EyeGlow
graph {
 properties {blendMode=BLEND_Translucent, pixelDepthOffsetMode=PDOM_Legacy, shadingModel=MSM_Unlit, twoSided=true, usedShadingModels=Unlit}
 nodes [
  n1 = OneMinus(None=VertexColor().R)
  n2 = TextureCoordinate()
  n3 = Power(Base=Sine(None=ComponentMask(None=@n2) {r=true}) {period=2}, Exp=ScalarParameter() {defaultValue=6, parameterName=Exp_TopFade})
  n4 = Multiply(
   A=TextureCoordinate()
   B=AppendVector(
    A=ScalarParameter() {defaultValue=0.05, parameterName=U_Scale_Opacity}
    B=ScalarParameter() {defaultValue=0.5, parameterName=V_Scale_Opacity}
   )
  )
  n5 = Multiply(
   A=StaticSwitchParameter(True=Constant(), False=Multiply(A=@n1, B=@n3)) {parameterName=Spline?}
   B=TextureSample(UVs=Panner(Coordinate=@n4) {speedX=0.05, speedY=0.1}) {texture=&T_NoisePack_01}.G
  )
  n6 = DepthFade(Opacity=@n5, FadeDistance=ScalarParameter() {defaultValue=10, parameterName=Fade Distance})
  FeatureLevelSwitch(Default=@n6, ES2=@n5, ES3_1=@n5, SM4_REMOVED=@n6)
  n7 = AppendVector(A=ScalarParameter() {defaultValue=0.5, parameterName=U_Scale_Red}, B=ScalarParameter() {defaultValue=0.05, parameterName=V_Scale_Red})
  n8 = TextureCoordinate()
  n9 = TextureSample(UVs=Panner(Coordinate=Multiply(A=@n8, B=@n7)) {speedY=0.075}) {texture=/Game/Effects/Fort_Effects/Textures/Plasma/T_PlasmaNoise_A_01}
  n10 = AppendVector(A=ScalarParameter() {defaultValue=0.25, parameterName=U_Scale_Blue}, B=ScalarParameter() {defaultValue=0.05, parameterName=V_Scale_Blue})
  n11 = TextureSample(UVs=Panner(Coordinate=Multiply(A=@n8, B=@n10)) {speedY=0.05}) {texture=&T_PlasmaNoise_A_01}
  n12 = LinearInterpolate(B=VectorParameter() {parameterName=EyeGlowColor, defaultValue=<rgba 0.8977 0.2243 1 1>}.RGB, Alpha=Clamp(None=Add(A=@n9.R, B=@n11.B)))
  n13 = PixelDepth()
  n14 = Constant()
  n15 = ScalarParameter() {defaultValue=500, parameterName=Radius}
  n16 = LinearInterpolate(
   A=ScalarParameter() {defaultValue=10, parameterName=Intensity_Close}
   B=ScalarParameter() {defaultValue=100, parameterName=Intensity_Far}
   Alpha=OneMinus(None=SphereMask(A=@n13.R, B=@n14, Radius=@n15) {hardnessPercent=0})
  )
  n17 = Multiply(A=ScalarParameter() {defaultValue=5, parameterName=Spline_Multi}, B=@n16)
  MaterialFunctionCall(Emissive=MaterialFunctionCall(LDR Input=Clamp(None=@n12), HDR Tint=@n12, HDR Intensity Multiplier=StaticSwitchParameter(True=@n17, False=@n16) {parameterName=Spline?}) {materialFunction=&DeriveHDRfromLDR}.RemappedRange) {materialFunction=&MF_FX_EmissiveAdjustments}
  n18 = TextureCoordinate()
  n19 = ComponentMask(None=TextureCoordinate()) {g=true}
  n20 = ScalarParameter() {defaultValue=1, parameterName=Exp_SplineEnd_Tip}
  n21 = OneMinus(None=@n19)
  n22 = ScalarParameter() {defaultValue=1.0788, parameterName=Exp_SplineEnd_Base}
  Multiply(
   A=Power(Base=ComponentMask(None=@n18) {r=true}, Exp=ScalarParameter() {defaultValue=3.5, parameterName=Exp_SplineSide})
   B=Multiply(A=Power(Base=@n19, Exp=@n20), B=Power(Base=@n21, Exp=@n22))
  )
  StaticBool()
  n23 = ScalarParameter() {defaultValue=1.4098, parameterName=SplineWidth_Base}
  n24 = ScalarParameter() {defaultValue=0.1712, parameterName=SplineWidth_Tip}
  n25 = MaterialFunctionCall!(
   Speed=ScalarParameter() {defaultValue=1, parameterName=Jiggle_Speed}
   LargeWaveScle=ScalarParameter() {defaultValue=1, parameterName=Jiggle_Small_Wave_Scale}
   WaveScale=ScalarParameter() {defaultValue=1, parameterName=Jiggle_Wave_Scale}
  ) {materialFunction=/Game/Effects/Fort_Effects/Materials/_Functions/Jiggle_WPO_VertColor_Red}
  n26 = Constant()
  n27 = Sine(None=Time()) {period=1.5}
  n28 = ScalarParameter() {defaultValue=-0.5, parameterName=Vertex_Offset_Min}
  n29 = ScalarParameter() {defaultValue=1, parameterName=Vertex_Offset_Max}
  n30 = ConstantBiasScale(None=Sine(None=Time()))
  n31 = Add(
   A=LinearInterpolate(
    A=VectorParameter() {parameterName=Z_Offset_Min, defaultValue=<rgba 0 0 -1 0>}.RGB
    B=VectorParameter() {parameterName=Z_Offset_Max, defaultValue=<rgba 0 0 1 0>}.RGB
    Alpha=ConstantBiasScale(None=@n27)
   ) {constAlpha=0}
   B=Multiply(A=VertexNormalWS(), B=LinearInterpolate(A=@n28, B=@n29, Alpha=@n30) {constAlpha=1})
  )
  n32 = TextureCoordinate()
  n33 = AppendVector(A=ScalarParameter() {defaultValue=0.1, parameterName=U_Scale_WPO}, B=ScalarParameter() {defaultValue=0.25, parameterName=V_Scale_WPO})
  n34 = TextureSample(UVs=Panner(Coordinate=Multiply(A=@n32, B=@n33)) {speedY=0.5}) {texture=&T_PlasmaNoise_A_01}
  StaticSwitchParameter(
   True=Add(
    A=MaterialFunctionCall(WidthBase=@n23, WidthTip=@n24) {materialFunction=&SplineThicken}.WorldPositionOffset
    B=AppendVector(A=@n25.Result, B=@n26)
   )
   False=Multiply(A=Multiply(A=@n31, B=@n34.G), B=VertexColor().R)
  ) {parameterName=Spline?}
 ]
}
